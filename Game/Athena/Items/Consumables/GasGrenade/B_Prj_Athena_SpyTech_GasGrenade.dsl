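; Blueprint for the SpyTech gas grenade projectile, a stink bomb variant used with the Infiltration playlist.
; It overrides the gas duration update event to look up a stink bomb duration row in the Infiltration OverrideGameData curve table.
; When the row is found, it sets the gas duration to that value, intended to be longer than the base, and the construction script only calls the parent.
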(event EventUpdateGasDuration
  ;[1 1 1] Override the base duration of stink bomb with a higher duration
  (bind _outxy (DataTable|EvaluateCurveTableRow "/Game/Athena/Playlists/Infiltration/OverrideGameData.OverrideGameData" "Infiltration.StinkBomb+.GasDuration")
    (:RowFound
      (Variables|GasGrenadeTuning|SetGasDuration _outxy))))

(fn ConstructionScript
  (|CallParentFunction))
