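# Translucent, lit material for the outer gas cloud of the gas grenade, tinted by particle color and time of day.
# Two panning noise layers blend two colors; their combined normals perturb a Fresnel that shapes opacity.
# Depth, camera-distance, time-of-day and Fade terms further scale opacity.
# The same normals drive vertex displacement, masked along V, faded by actor distance and disabled on mobile.
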Material /Game/Athena/Items/Consumables/GasGrenade/M_Athena_GasGrenade_Outer_01
graph {
 properties {
  bEnableResponsiveAA=true
  blendMode=BLEND_Translucent
  pixelDepthOffsetMode=PDOM_Legacy
  translucencyLightingMode=TLM_VolumetricPerVertexDirectional
  usedShadingModels=Default Lit
 }
 nodes [
  AppendVector()
  ComponentMask() {g=true, r=true}
  n1 = ScalarParameter() {defaultValue=0.075, parameterName=Hue Shift Morning}
  n2 = ScalarParameter() {defaultValue=1, parameterName=U1}
  n3 = ScalarParameter() {defaultValue=1.3, parameterName=V1}
  n4 = Panner(Coordinate=Multiply(A=TextureCoordinate(), B=AppendVector(A=@n2, B=@n3))) {bFractionalPart=true, speedY=0.02}
  n5 = ScalarParameter() {defaultValue=2, parameterName=U2}
  n6 = ScalarParameter() {defaultValue=1.125, parameterName=V2}
  n7 = Panner(Coordinate=Multiply(A=TextureCoordinate(), B=AppendVector(A=@n5, B=@n6))) {bFractionalPart=true, speedY=0.0672}
  n8 = Divide(A=Add(
    A=TextureSample(UVs=@n4) {texture=&2ndInfectedMapLinear}.R
    B=TextureSample(UVs=@n7) {texture=&2ndInfectedMapLinear}.R
   ))
  n9 = ComponentMask(None=VertexNormalWS()) {b=true}
  n10 = Clamp(None=Power(Base=Abs(None=@n9)) {constExponent=18})
  n11 = LinearInterpolate(
   A=VectorParameter() {parameterName=Color2, defaultValue=<rgba 0.3229 0.2534 0.1827 0>}.RGB
   B=VectorParameter() {parameterName=Color1, defaultValue=<rgba 2 1.1212 0.2812 0>}.RGB
   Alpha=Clamp(None=LinearInterpolate(A=@n8, Alpha=@n10) {constB=0.5})
  ) {constA=0.5}
  n12 = ScalarParameter() {defaultValue=0.1, parameterName=Hue Shift Evening}
  n13 = ScalarParameter() {defaultValue=0.05, parameterName=Hue Shift Night}
  n14 = ParticleColor()
  Multiply(
   A=MaterialFunctionCall(
    Color_Morning=MaterialFunctionCall(Hue Shift Percentage=@n1, Texture=@n11) {materialFunction=&HueShift}.Result
    Color_Day=@n11
    Color_Evening=MaterialFunctionCall(Hue Shift Percentage=@n12, Texture=@n11) {materialFunction=&HueShift}.Result
    Color_Night=MaterialFunctionCall(Hue Shift Percentage=@n13, Texture=@n11) {materialFunction=&HueShift}.Result
   ) {materialFunction=&Color_Time_of_Day}.Result
   B=@n14.RGB
  ) {constA=1}
  n15 = ScalarParameter() {defaultValue=1.5, parameterName=FresnelBrightness}
  n16 = ScalarParameter() {defaultValue=2.5, parameterName=FresnelExponent}
  n17 = Normalize(VectorInput=Add(
    A=TextureSample(UVs=@n4) {samplerType=SAMPLERTYPE_Normal, texture=/Game/Global/Textures/Default/Tileable_Patterns/2ndInfectedMapNormal}.RGB
    B=TextureSample(UVs=@n7) {samplerType=SAMPLERTYPE_Normal, texture=&2ndInfectedMapNormal}.RGB
   ))
  n18 = VectorParameter() {parameterName=NormalIntensity, defaultValue=<rgba 0.3 0.3 1.5 0>}
  n19 = LinearInterpolate(
   A=Transform(None=Multiply(A=@n17, B=@n18.RGB))
   B=VertexNormalWS()
   Alpha=MaterialFunctionCall(Exponent=ScalarParameter() {defaultValue=3, parameterName=2ndFresnelLerp}) {materialFunction=&MF_Fresnel_FX}.Fresnel
  )
  n20 = ScalarParameter() {parameterName=FogBias}
  n21 = ScalarParameter() {defaultValue=75, parameterName=FogDivisor}
  n22 = Multiply(
   A=@n14.A
   B=Multiply(
    A=MaterialFunctionCall(Brightness=@n15, Exponent=@n16, Vertex / Pixel Normal=@n19) {materialFunction=&MF_Fresnel_FX}.Fresnel
    B=MaterialFunctionCall!(Distance Bias=@n20, Distance Divisor=@n21, Scene Depth, Pixel Depth) {materialFunction=/Game/Effects/Fort_Effects/Materials/_Functions/MF_DistanceBehindTranslucency_01}.Normalized
   )
  )
  n23 = WorldPosition()
  n24 = ScalarParameter() {defaultValue=40, parameterName=CameraFadeDistanceStart}
  n25 = ScalarParameter() {defaultValue=15, parameterName=CameraFadeDistanceEnd}
  n26 = ScalarParameter() {defaultValue=0.75, parameterName=OpacityMult}
  n27 = ScalarParameter() {defaultValue=0.75, parameterName=OpacityMult_Night}
  n28 = ScalarParameter() {defaultValue=1, parameterName=Fade}
  Multiply(
   A=Multiply(
    A=MaterialFunctionCall(Opacity In=@n22, Position=@n23.XYZ, Fade_Start_Distance=@n24, Fade_End_Distance=@n25) {materialFunction=&MF_Distance_SyncedFade_FX}.Result
    B=MaterialFunctionCall(Color_Morning=@n26, Color_Day=@n26, Color_Evening=@n26, Color_Night=@n27) {materialFunction=&Color_Time_of_Day}.Result
   )
   B=@n28
  )
  n29 = VertexNormalWS()
  n30 = Multiply(
   A=VectorParameter() {parameterName=WPO Scale, defaultValue=<rgba 90 90 0 0>}.RGB
   B=MaterialFunctionCall(
    Position=ActorPositionWS()
    Fade_Start_Distance=ScalarParameter() {defaultValue=2000, parameterName=WPO Fade Start}
    Fade_End_Distance=ScalarParameter() {defaultValue=900, parameterName=WPO Fade End}
   ) {materialFunction=&MF_Distance_SyncedFade_FX}.Result
  )
  n31 = SphereMask(
   A=ComponentMask(None=TextureCoordinate()) {g=true}
   B=Constant() {r=0.5}
   Radius=ScalarParameter() {defaultValue=0.45, parameterName=WPO Mask Radius}
   Hardness=ScalarParameter() {defaultValue=0.7, parameterName=WPO Mask Hardness}
  )
  ShadingPathSwitch(Default=Multiply(A=Multiply(A=@n17, B=@n29), B=Multiply(A=@n30, B=@n31)), Mobile=Constant()) {desc=Migrated from FeatureLevelSwitch}
  SphereMask(
   A=TextureCoordinate() {uTiling=0, vTiling=1.5}
   B=Constant2Vector() {g=1}
   Radius=Multiply(A=@n28)
   Hardness=ScalarParameter() {defaultValue=0.25, parameterName=Smoke_Fadein_Hardness}
  )
 ]
}
