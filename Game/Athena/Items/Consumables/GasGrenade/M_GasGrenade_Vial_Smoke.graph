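# Translucent unlit material for the gas grenade vial, faking swirling smoke inside glass.
# Two panned cloud height and normal layers drive smoke color, edge glow and opacity, faded along V.
# A view-space reflection lookup, sun/moon highlight and time-of-day tint add glass shine over Fresnel edges, while UV sphere masks shade the top separately.
# World position offset combines LOD/pickup distance-cull and black hole functions.
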Material /Game/Athena/Items/Consumables/GasGrenade/M_GasGrenade_Vial_Smoke
graph {
 properties {
  allowTranslucentCustomDepthWrites=true
  bUsedWithMeshDeformer=true
  bUsedWithSkeletalMesh=true
  blendMode=BLEND_Translucent
  pixelDepthOffsetMode=PDOM_Legacy
  shadingModel=MSM_Unlit
  usedShadingModels=Unlit
 }
 nodes [
  n1 = SphereMask(
   A=TextureCoordinate() {uTiling=0}
   B=Constant2Vector() {g=0.1}
   Radius=ScalarParameter() {defaultValue=0.1, parameterName=RadiusTopMask}
   Hardness=ScalarParameter() {defaultValue=0.9, parameterName=HardnessTopMask}
  )
  n2 = ComponentMask(None=Transform(None=ReflectionVectorWS()) {transformSourceType=TRANSFORMSOURCE_World, transformType=TRANSFORM_View}) {g=true, r=true}
  n3 = Constant2Vector() {g=-0.4, r=0.4}
  n4 = TextureSampleParameter2D(UVs=Add(A=Multiply(A=@n2, B=@n3), B=Constant2Vector() {g=0.5, r=0.5})) {parameterName=SphereReflectionMap, texture=/Game/Items/Art_noLOD/Textures/T_Items_MineralPowder_Glass_Reflection}
  n5 = CollectionParameter() {
   collection=&FortniteMaterialParameters
   parameterName=SunAndMoonModelDirectionalVector
  }
  n6 = ComponentMask(None=@n5) {b=true, g=true, r=true}
  n7 = ReflectionVectorWS()
  n8 = Power(Base=Clamp(None=DotProduct(A=@n6, B=@n7)), Exp=ScalarParameter() {defaultValue=70, parameterName=Spec Highlight Power}) {constExponent=20}
  n9 = ScalarParameter() {defaultValue=3, parameterName=Spec Intense}
  n10 = Multiply(
   A=Add(A=MaterialFunctionCall() {materialFunction=&Color_Time_of_Day}.Result) {constB=0.5}
   B=Add(A=Multiply(A=@n4.RGB) {constB=0.25}, B=Multiply(A=@n8, B=@n9))
  )
  n11 = Fresnel() {baseReflectFraction=0, exponent=1}
  n12 = Multiply(A=@n11, B=@n11)
  n13 = ComponentMask(None=TextureCoordinate()) {g=true}
  n14 = Time()
  n15 = Panner(Coordinate=TextureCoordinate(), Time=@n14) {speedX=0.075, speedY=0.105}
  n16 = Panner(Coordinate=TextureCoordinate(), Time=@n14) {speedX=-0.015, speedY=0.09}
  n17 = LinearInterpolate(
   A=TextureSample(UVs=@n15) {samplerType=SAMPLERTYPE_LinearGrayscale, texture=/Game/Environments/Hexmap/HexClouds/T_HexmapFog_01_Height_GS}.R
   B=TextureSample(UVs=@n16) {samplerType=SAMPLERTYPE_LinearGrayscale, texture=/Game/Global/Textures/Default/Tileable_Patterns/T_StormCloudDetail01_M}.R
  ) {constAlpha=0.4}
  n18 = LinearInterpolate(Alpha=@n17) {constA=-0.1}
  n19 = LinearInterpolate(
   A=TextureSample(UVs=@n15) {samplerType=SAMPLERTYPE_Normal, texture=/Game/Global/Textures/Default/Tileable_Patterns/T_StormCloudDetail01_N}.RGB
   B=TextureSample(UVs=@n16) {samplerType=SAMPLERTYPE_Normal, texture=&T_StormCloudDetail01_N}.RGB
  )
  n20 = DotProduct(A=Transform(None=@n19), B=CameraVectorWS())
  n21 = Multiply(A=Multiply(A=Clamp(None=@n18), B=Multiply(A=@n20, B=@n20)), B=ScalarParameter() {defaultValue=4, parameterName=OpacityMult}) {constB=4}
  n22 = Multiply(A=Multiply(A=@n13, B=@n13), B=Multiply(A=Clamp(None=@n21) {maxDefault=2}, B=VertexColor().R))
  Add(A=@n1, B=Add(A=@n10, B=Add(A=@n12, B=@n22)))
  n23 = VectorParameter() {parameterName=WorldPostionOfBlackHole, defaultValue=<rgba -30000 25536 1952 0>}
  n24 = MaterialFunctionCall!(
   World Position
   DistanceCullFadeAnimationOffset=MaterialFunctionCall() {materialFunction=/Game/Packages/Fortress_SharedMaterials/Base_Material_Functions/PickupEffectDistanceCullOverride}.Result
   Use actor center instead of pivot point
  ) {materialFunction=/Game/Packages/Fortress_SharedMaterials/Base_Material_Functions/LODEffect}
  n25 = WorldPosition()
  Add(
   A=MaterialFunctionCall!(WorldPostionOfBlackHole=AppendVector(A=@n23.RGB, B=@n23.RGB), World Position=Add(A=@n24.Result, B=@n25.XYZ)) {materialFunction=/Game/Packages/Fortnite_UniqueMaterials_1/BlackHoleFunction}.Result
   B=@n24.Result
  )
  n26 = VectorParameter() {parameterName=Color2, defaultValue=<rgba 0.3229 0.2534 0.1827 0>}
  n27 = Constant() {r=0.7}
  n28 = Fresnel(Normal=Transform(None=MaterialFunctionCall(Normal=@n19, Flatness=@n27) {materialFunction=&FlattenNormal}.Result)) {baseReflectFraction=0, exponent=0.1}
  n29 = Multiply(A=Clamp(None=LinearInterpolate(Alpha=@n28) {constA=2, constB=-0.25}), B=@n17)
  n30 = Add(
   A=LinearInterpolate(A=@n26.RGB, B=VectorParameter() {parameterName=SmokeColor2, defaultValue=<rgba 0.0354 0.04 0.0163 0>}.RGB, Alpha=@n17)
   B=Multiply(A=@n26.RGB, B=Multiply(A=@n29) {constB=15})
  )
  n31 = VectorParameter() {parameterName=Color1, defaultValue=<rgba 2 1.1212 0.2812 0>}
  n32 = Power(Base=@n20) {constExponent=4}
  n33 = Multiply(A=@n31.RGB, B=Add(A=@n32) {constB=0.05})
  n34 = ScalarParameter() {defaultValue=1, parameterName=EdgeGlowMax}
  n35 = ScalarParameter() {defaultValue=-0.5, parameterName=EdgeGlowMin}
  n36 = Multiply(A=Clamp(None=LinearInterpolate(A=@n34, B=@n35, Alpha=@n17) {constA=1, constB=-0.3}), B=@n32)
  n37 = Multiply(A=Add(A=Add(A=@n30, B=@n33), B=Multiply(A=@n31.RGB, B=@n36)), B=@n13)
  n38 = VectorParameter() {parameterName=GlassEdgeColor, defaultValue=<rgba 0.1771 0.1657 0.1171 1>}
  n39 = TextureSample(UVs=MaterialFunctionCall(
    UVs=TextureCoordinate() {uTiling=35, vTiling=0.5}
    "Rotation Angle (0-1)"=ScalarParameter() {defaultValue=0.25, parameterName="Rotation Angle (0-1) (S)"}
   ) {materialFunction=&CustomRotator}.Rotated Values) {texture=/Game/Effects/Fort_Effects/Textures/Grad/T_Gradient_Rainbow_NoRed}
  n40 = VectorParameter() {parameterName=GlassTopColor, defaultValue=<rgba 0.008 0.008 0.008 1>}
  n41 = Multiply(A=Desaturation(None=ComponentMask(None=@n39.RGB) {g=true}), B=@n40.RGB)
  n42 = TextureSample(UVs=TextureCoordinate() {uTiling=0.5, vTiling=35}) {texture=&T_Gradient_Rainbow_NoRed}
  n43 = Multiply(A=Desaturation(None=ComponentMask(None=@n42.RGB) {g=true}), B=@n40.RGB)
  n44 = SphereMask(
   A=TextureCoordinate() {uTiling=0}
   B=Constant2Vector() {g=0.05}
   Radius=ScalarParameter() {defaultValue=0.05, parameterName=RadiusTopMask_1}
   Hardness=ScalarParameter() {defaultValue=0.9, parameterName=HardnessTopMask_1}
  )
  LinearInterpolate(
   A=Add(A=@n10, B=LinearInterpolate(A=@n37, B=@n38.RGB, Alpha=@n12))
   B=Add(A=Multiply(A=@n10) {constB=0.5}, B=LinearInterpolate(A=@n41, B=@n43, Alpha=@n44))
   Alpha=@n1
  )
 ]
}
